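; QA helper Blueprint that spawns BP_VolumeVisualizer actors to show the Blocking and KillZ volumes in a map.
; Separate custom events handle each volume type, and a third destroys every visualizer.
; Each spawn temporarily zeroes the volume's rotation to size the visualizer from its bounds extent times a multiplier, then restores that rotation on both actors.
; CheckDoesVisualizerExist prevents duplicates by comparing existing visualizer locations with the volume's bounds origin.
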(event Custom|SpawnBlockingVolumeVisualizers
  ;[1 1 1] Blocking Volume Spawn
  (bind _outactors (Actor|GetAllActorsOfClass "/Script/Engine.BlockingVolume"))
  (for _array_element _outactors
    ;;[1 1 1] make sure we don't make duplicate visualizers if the function is ran multiple times
    (bind _doesexist (CallFunction|CheckDoesVisualizerExist (Collision|GetComponentBounds (Variables|/Script/Engine.Brush|GetBrushComponent _array_element))))
    (if (not _doesexist)
      ;;[1 1 1] store the original rotation of the brush
      (Variables|Default|SetOriginalRotation (Transformation|GetActorRotation _array_element))
      ;;[1 1 1] zero out the rotation so we can make a properly sized visualizer
      (Transformation|SetActorRotation _array_element "0, 0, 0" false)
      ;;[1 1 1] create the visualizer then set it's rotation to match the brush
      (bind _returnvalue (Game|SpawnActorfromClass "/Game/QA/Map_QA/Milson/KillZ_Visualizer/BP_VolumeVisualizer.BP_VolumeVisualizer_C" (Math|Transform|MakeTransform (Collision|GetComponentBounds (Variables|/Script/Engine.Brush|GetBrushComponent _array_element)) "0, 0, 0" (* (.BoxExtent (Collision|GetComponentBounds (Variables|/Script/Engine.Brush|GetBrushComponent _array_element))) (Variables|Default|Getmult_amnt))) "Undefined" "OverrideRootScale" 0 0))
      (Transformation|SetActorRotation _returnvalue (Variables|Default|GetOriginalRotation) false)
      ;;[1 1 1] reset the brush rotation to what it was
      (Transformation|SetActorRotation _array_element (Variables|Default|GetOriginalRotation) false))))

(event Custom|SpawnKillZVisualizers
  ;[1 1 1] Blocking Volume Spawn
  (bind _outactors (Actor|GetAllActorsOfClass "/Script/Engine.KillZVolume"))
  (for _array_element _outactors
    ;;[1 1 1] make sure we don't make duplicate visualizers if the function is ran multiple times
    (bind _doesexist (CallFunction|CheckDoesVisualizerExist (Collision|GetComponentBounds (Variables|/Script/Engine.Brush|GetBrushComponent _array_element))))
    (if (not _doesexist)
      ;;[1 1 1] store the original rotation of the brush
      (Variables|Default|SetOriginalRotation (Transformation|GetActorRotation _array_element))
      ;;[1 1 1] zero out the rotation so we can make a properly sized visualizer
      (Transformation|SetActorRotation _array_element "0, 0, 0" false)
      ;;[1 1 1] create the visualizer then set it's rotation to match the brush
      (bind _returnvalue (Game|SpawnActorfromClass "/Game/QA/Map_QA/Milson/KillZ_Visualizer/BP_VolumeVisualizer.BP_VolumeVisualizer_C" (Math|Transform|MakeTransform (Collision|GetComponentBounds (Variables|/Script/Engine.Brush|GetBrushComponent _array_element)) "0, 0, 0" (* (.BoxExtent (Collision|GetComponentBounds (Variables|/Script/Engine.Brush|GetBrushComponent _array_element))) (Variables|Default|Getmult_amnt))) "Undefined" "OverrideRootScale" 0 0))
      (Transformation|SetActorRotation _returnvalue (Variables|Default|GetOriginalRotation) false)
      ;;[1 1 1] reset the brush rotation to what it was
      (Transformation|SetActorRotation _array_element (Variables|Default|GetOriginalRotation) false))))

(event Custom|RemoveAllVisualizers
  ;[1 1 1] Remove all Volume Visualizer BPs placed in the world
  (bind _outactors (Actor|GetAllActorsOfClass "/Game/QA/Map_QA/Milson/KillZ_Visualizer/BP_VolumeVisualizer.BP_VolumeVisualizer_C"))
  (for _array_element _outactors
    (Actor|DestroyActor _array_element)))

(fn CheckDoesVisualizerExist
  (params (VolumeLocation :type Vector))
  (Variables|Default|Setvisualizer_exists false)
  (bind _outactors (Actor|GetAllActorsOfClass "/Game/QA/Map_QA/Milson/KillZ_Visualizer/BP_VolumeVisualizer.BP_VolumeVisualizer_C"))
  ;[1 1 1] go through every existing visualizer and see if one already exists at the passed in KillZ volume's location
  (bind _array_element (Utilities|Array|ForEachLoopwithBreak _outactors)
    (:LoopBody
      (bind _returnvalue_location (Math|Transform|BreakTransform (Transformation|GetActorTransform _array_element)))
      (if (Math|Vector|Equal(Vector) _returnvalue_location VolumeLocation 0.0001)
        (Variables|Default|Setvisualizer_exists true)))
    (:Completed
      _))
  (return (Variables|Default|Getvisualizer_exists)))

(fn ConstructionScript)
